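# Translucent unlit Niagara sprite and mesh particle material, named for the rain wave effect of the Melancholy emote.
# Two smoke-trail samples, offset per particle and distorted by panning Perlin noise, blend a cool-tinted particle color toward the plain one for emissive.
# Opacity comes from Voronoi noise eroded by particle alpha inside a sine edge mask, scaled by a Particle_Opacity dynamic parameter and view-angle fade.
# Depth fade softens intersections except on the mobile shading path.
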Material /BRCosmetics/Animation/Game/MainPlayer/Emotes/Melancholy/VFX/M_Emote_Melancholy_RainWave
graph {
 properties {
  bUsedWithNiagaraMeshParticles=true
  bUsedWithNiagaraSprites=true
  blendMode=BLEND_Translucent
  pixelDepthOffsetMode=PDOM_Legacy
  shadingModel=MSM_Unlit
  usedShadingModels=Unlit
 }
 nodes [
  n1 = Constant3Vector() {constant=<rgba 0.456 0.504 0.57 1>}
  n2 = ParticleColor()
  n3 = ParticleRandom()
  n4 = Multiply(A=TextureCoordinate()) {constB=0.5}
  n5 = TextureSampleParameter2D(UVs=Panner(Coordinate=Add(A=@n3, B=@n4)) {speedX=0.3, speedY=-0.2}) {
   parameterName=Noise1
   samplerType=SAMPLERTYPE_Masks
   texture=/Game/Characters/Player/Male/Large/Bodies/M_LRG_Clash_V/Textures/Simple_Perlin_M
  }
  n6 = Multiply(A=@n5.R) {constB=0.1}
  n7 = MaterialFunctionCall(Float2=TextureCoordinate()) {materialFunction=&BreakOutFloat2Components}
  n8 = Add(A=ParticleRandom(), B=Multiply(A=@n7.R) {constB=0.7})
  n9 = TextureSampleParameter2D(UVs=Add(A=@n6, B=AppendVector(A=@n8, B=@n7.G))) {parameterName=Tex1, texture=/Game/Effects/Fort_Effects/Textures/Smoke/T_Smoke_Trail_Soft_2}
  n10 = Add(A=@n7.G) {constB=0.4}
  n11 = TextureSampleParameter2D(UVs=Add(A=@n6, B=AppendVector(A=@n8, B=@n10))) {parameterName=Tex1, texture=&T_Smoke_Trail_Soft_2}
  MaterialFunctionCall(Emissive=LinearInterpolate(A=Multiply(A=@n1, B=@n2.RGB), B=@n2.RGB, Alpha=LinearInterpolate(A=@n9.G, B=@n11.R) {constAlpha=0.2})) {materialFunction=&MF_FX_EmissiveAdjustments}
  n12 = MaterialFunctionCall(Float2=TextureCoordinate()) {materialFunction=&BreakOutFloat2Components}
  n13 = AppendVector(A=Multiply(A=@n12.R), B=Multiply(A=@n12.G) {constB=0.5})
  n14 = TextureCoordinate()
  n15 = Multiply(A=ComponentMask(None=@n14) {r=true}) {constB=0.5}
  n16 = Multiply(A=ComponentMask(None=@n14) {g=true}) {constB=0.5}
  n17 = Multiply(A=Power(Base=Sine(None=@n15)) {constExponent=1.3}, B=Power(Base=Sine(None=@n16)) {constExponent=1.3})
  n18 = Multiply(
   A=Subtract(
    A=TextureSampleParameter2D(UVs=@n13) {parameterName=DissolveNoise, texture=/Game/Effects/Fort_Effects/Textures/T_PinkSpike_Voronoi2D_Height}.R
    B=LinearInterpolate(Alpha=@n2.A) {constA=1, constB=-1}
   )
   B=Multiply(A=Saturate(None=@n17)) {constB=0.4}
  )
  n19 = DepthFade(Opacity=@n18) {fadeDistanceDefault=22}
  Multiply(
   A=Multiply(
    A=ShadingPathSwitch(Default=@n19, Deferred=@n19, Mobile=@n18) {desc=Migrated from FeatureLevelSwitch}
    B=DynamicParameter() {paramNames [Particle_Opacity, Param2, Param3, Param4]}.Particle_Opacity
   )
   B=MaterialFunctionCall("FadeSensitivity (optional)"=Constant() {r=1}) {materialFunction=&FadeBasedOnViewAngle}.Result
  )
  VertexInterpolator()
 ]
}
