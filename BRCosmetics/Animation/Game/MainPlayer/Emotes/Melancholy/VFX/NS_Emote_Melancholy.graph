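# Niagara effect for the Melancholy emote: a cloud that fades in and rains.
# A camera-facing mesh cloud, with puff sprites, changes shade along a system timeline.
# Rain, camera-relative side rain and a thick curtain fall beneath, with time-of-day tints on curtain, puffs and splashes.
# User.UmbrellaUp shortens central rain and curtain lifetimes, raises their spawn rates and enables impact splashes; the Rain_Sides emitter is disabled.
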NiagaraSystem /BRCosmetics/Animation/Game/MainPlayer/Emotes/Melancholy/VFX/NS_Emote_Melancholy
dependencies {
 usedDataInterfaces [
  &NiagaraDataInterfaceCamera
  &NiagaraDataInterfaceColorCurve
  &NiagaraDataInterfaceCurve
  &NiagaraDataInterfaceVector2DCurve
  &NiagaraDataInterfaceSpriteRendererInfo
  &NiagaraDataInterfaceStaticMesh
  &NiagaraDataInterfaceParticleRead
  &NiagaraDataInterfaceArrayFloat3
  &NiagaraDataInterfaceArrayFloat
 ]
 usedDynamicInputs [
  &ReturnCameraPosition
  &MakeFloatFromVector
  &SetBoolByFloatComparison
 ]
 usedModules [
  &SystemState
  &self:SystemSpawnScript.SystemScriptSource.SystemScriptGraph.NiagaraNodeAssignment_4.SetVariables_#1
  &self:SystemSpawnScript.SystemScriptSource.SystemScriptGraph.NiagaraNodeAssignment_3.SetVariables_#2
  &self:SystemSpawnScript.SystemScriptSource.SystemScriptGraph.NiagaraNodeAssignment_2.SetVariables_#3
  &self:SystemSpawnScript.SystemScriptSource.SystemScriptGraph.NiagaraNodeAssignment_0.SetVariables_#4
  &Timeline
 ]
}
emitters {
 CloudPuffsExtra {
  emitterUpdateScript [
   &EmitterState()
   &SpawnPerUnit(Module.Use Max Movement Threshold=true, Module.Use Movement Tolerance=false, Spawn Spacing=50, Use Spawn Probability=false)
   &SpawnRate(SpawnRate=4, Use Spawn Probability=false) {enabled=false}
  ]
  particleSpawnScript [
   &InitializeParticle(
    Color=&FNTimeOfDayColor(
     Night Color=<rgba 0.507 0.5253 0.65 1>
     Morning Color=<rgba 0.825 0.65 1 1>
     Day Color=<rgba 0.72 0.8367 1 1>
     Evening Color=<rgba 1 0.7621 0.52 1>
    )
    Lifetime Max=3
    Lifetime Min=2
    Lifetime Mode=::NewEnumerator1
    Sprite Rotation Mode=::NewEnumerator2
    Sprite Size=&Multiply_Vector2D_ByFloat(Float=&RandomRangeFloat(Minimum=40, Maximum=60), Vector2D=<1 1>)
    Sprite Size Mode=::NewEnumerator4
    UsePositionOffset=false
    Write Material Random=true
   )
   &ShapeLocation(
    MaskSpawnGroup=false
    Non Uniform Scale=<1 2.5 0.6>
    Offset=<0 0 16>
    Offset Mode=::NewEnumerator0
    Override Seed=false
    Sphere Radius=10
    Sphere Surface Distribution=1
    UseOwnerScale=true
   )
   &CameraOffset(Camera Offset Amount=70)
   &AddVelocity(Velocity=<-50 0 0>) {enabled=false}
   &SubUVAnimation(
    "Sprite Renderer "=DataInterface() {propertyValues {SpriteRenderer=&self:CloudPuffsExtra_2.NiagaraSpriteRendererProperties_0}}
    Start Frame Range Override=32
    SubUV Animation Mode:ENiagara_SubUVLookupModeV2=::NewEnumerator2
    UseEndFrame=true
    UseStartFrame=true
   ) {module=SubUVAnimation001}
  ]
  particleUpdateScript [
   &ParticleState()
   &ScaleVelocity(Velocity Scale=&MakeVector(
     X=&FloatFromCurve(FloatCurve=DataInterface() {propertyValues {Curve {keys [<at 0 1>,<at 1 0>]}}})
     Y=0
     Z=0
    )) {enabled=false}
   &InheritSourceMovement(
    Applied Position Scale=&MakeVector(
     X=&FloatFromCurve(FloatCurve=DataInterface() {propertyValues {Curve {keys [{interpMode=RCIM_Cubic, tangentMode=RCTM_User, time=0.3, value=1}, {interpMode=RCIM_Cubic, tangentMode=RCTM_User, time=0.8, value=0.1}]}}})
     Y=&FloatFromCurve(FloatCurve=DataInterface() {propertyValues {Curve {keys [{interpMode=RCIM_Cubic, tangentMode=RCTM_User, time=0.3, value=1}, {interpMode=RCIM_Cubic, tangentMode=RCTM_User, time=0.8, value=0.1}]}}})
     Z=0
    )
    ApplyForce=false
    ApplyPosition=true
    ApplyPositionFalloff=false
   )
   &ScaleSpriteSize(Uniform Curve Sprite Scale=DataInterface() {propertyValues {Curve {keys [{interpMode=RCIM_Cubic, tangentMode=RCTM_User, time=0, value=0.7}, {interpMode=RCIM_Cubic, tangentMode=RCTM_User, time=1, value=1}]}}})
   &CurlNoiseForce(Noise Frequency=20, Noise Strength=&RandomRangeFloat(Minimum=20, Maximum=40), Pan Noise=false) {enabled=false}
   &Drag(Drag=2, Use Linear Drag=true, Use Rotational Drag=true)
   &ScaleColor(
    Scale Alpha=&FloatFromCurve(
     FloatCurve=DataInterface() {propertyValues {Curve {keys [
         <at 0 0 RCIM_Cubic>
         {arriveTangent=-0.0046, interpMode=RCIM_Cubic, leaveTangent=-0.0046, tangentMode=RCTM_User, time=0.4, value=1}
         {arriveTangent=-4.3745, interpMode=RCIM_Cubic, leaveTangent=-4.3745, tangentMode=RCTM_User, time=0.5, value=0.8}
         <at 1 0 RCIM_Cubic>
        ]}}}
     Scale Curve=&RandomRangeFloat(Minimum=0.5, Maximum=1.5)
    )
    Scale RGB=&MakeVectorFromLinearColorRGB(Color=&ColorFromCurve(
      ColorCurve=DataInterface() {propertyValues {
        AlphaCurve {keys []}
        BlueCurve {keys [<at 0 0.4 RCIM_Cubic>,<at 0.1739 0.1>,<at 0.5892 0.1>,<at 1 0.4 RCIM_Cubic>]}
        GreenCurve {keys [<at 0 0.4 RCIM_Cubic>,<at 0.1739 0.1>,<at 0.5892 0.1>,<at 1 0.4 RCIM_Cubic>]}
        RedCurve {keys [<at 0 0.4 RCIM_Cubic>,<at 0.1739 0.1>,<at 0.5892 0.1>,<at 1 0.4 RCIM_Cubic>]}
       }}
      CurveIndex=$System.Timeline001.NormalizedTime
     ))
    ScaleA=true
    ScaleRGB=true
   )
   &SubUVAnimation(
    End Frame Range Override=64
    "Sprite Renderer "=DataInterface() {propertyValues {SpriteRenderer=&self:CloudPuffsExtra_2.NiagaraSpriteRendererProperties_0}}
    Start Frame Range Override=32
    UseEndFrame=true
    UseStartFrame=true
   ) {enabled=false}
   &SolveForcesAndVelocity(Clamp Velocity=false, Limit Acceleration=false)
  ]
  rendererClasses [&NiagaraSpriteRendererProperties]
  renderers [{rendererClass=&NiagaraSpriteRendererProperties}]
 }
 Cloud_Main {
  emitterUpdateScript [
   &EmitterState(Life Cycle Mode=::NewEnumerator1, Loop Behavior=::NewEnumerator1, Loop Delay=0.15, Loop Duration Mode=::NewEnumerator1, UseLoopDelay=true)
   &SpawnBurst_Instantaneous(Use Spawn Probability=false, UseLoopCountLimit=true)
  ]
  particleSpawnScript [
   &InitializeParticle(
    Lifetime=$System.CloudDuration
    Mesh Scale Mode=::NewEnumerator3
    Mesh Uniform Scale=0.8
    UsePositionOffset=false
    Write Material Random=true
   )
   &DynamicMaterialParameters(
    Index 0 Param 1=20
    Index 0 Param 2=-10
    Index 0 Param 3=2
    Index 0 Param 4=2
    Module.Param0WriteEnabled=true
    Module.Param1WriteEnabled=true
    Module.Param2WriteEnabled=true
    Module.Param3WriteEnabled=true
   )
   &InitialMeshOrientation(Mesh Axis To Orient=<0 1 0>, Mesh Orientation Mode=::NewEnumerator2, Rotation=<0.05 0 0>, Use Rotation Vector=true)
  ]
  particleUpdateScript [
   &ParticleState(Kill Particles When Lifetime Has Elapsed=false, Let Infinitely Lived Particles Die When Emitter Deactivates=true)
   &UpdateMeshOrientation(Facing Interpolation=false, Orientation Method=::NewEnumerator1, Override Quat=false, TransformFacing=false, TransformUp=false) {enabled=false}
   &InterpolateOverTime(
    Rate Of Change:NiagaraFloat=&Lerp_Float(
     A=100
     B=2
     Alpha=&DotProduct(
      Vector A=&ReturnCameraVector(Camera Vectors=::NewEnumerator4, Coordinate Space=::Simulation, Origin Position=$Particles.Position)
      Vector B=<0 0 1>
     )
    )
    Target Vector Value:Vector3f=&ReturnCameraVector(Camera Vectors=::NewEnumerator1, Coordinate Space=::Simulation)
    Variable Type=::NewEnumerator2
   )
   &UpdateMeshOrientation(
    Facing Direction=$Particles.InterpolateOverTime.InterpolateOverTime_V3_.Moving Average
    Facing Interpolation=false
    Orientation Method=::NewEnumerator1
    Override Quat=false
    TransformFacing=false
    TransformUp=false
   ) {module=UpdateMeshOrientation001}
   /SetVariables(Particles.Position=$System.PositionCloud) {isSetParameters=true}
   &ScaleColor(
    Scale Alpha=&FloatFromCurve(
     FloatCurve=DataInterface() {propertyValues {Curve {keys [
         {arriveTangent=0.0328, interpMode=RCIM_Cubic, leaveTangent=0.0328, tangentMode=RCTM_User, time=0, value=0}
         {arriveTangent=-0.045, interpMode=RCIM_Cubic, leaveTangent=-0.045, tangentMode=RCTM_User, time=0.5, value=1}
        ]}}}
     CurveIndex=$Emitter.Age
    )
    Scale RGB=&MakeVectorFromLinearColorRGB(Color=&ColorFromCurve(
      ColorCurve=DataInterface() {propertyValues {
        AlphaCurve {keys []}
        BlueCurve {keys [
          <at 0 0.5 RCIM_Cubic>
          {interpMode=RCIM_Cubic, tangentMode=RCTM_User, time=0.143, value=0.3}
          {arriveTangent=0.2382, interpMode=RCIM_Cubic, leaveTangent=0.2382, time=0.5259, value=0.35}
          <at 0.9826 0.5>
         ]}
        GreenCurve {keys [
          <at 0 0.5 RCIM_Cubic>
          {interpMode=RCIM_Cubic, tangentMode=RCTM_User, time=0.143, value=0.2588}
          {arriveTangent=0.2873, interpMode=RCIM_Cubic, leaveTangent=0.2873, time=0.5259, value=0.302}
          <at 0.9826 0.5>
         ]}
        RedCurve {keys [
          <at 0 0.5 RCIM_Cubic>
          {interpMode=RCIM_Cubic, tangentMode=RCTM_User, time=0.143, value=0.2412}
          {arriveTangent=0.3083, interpMode=RCIM_Cubic, leaveTangent=0.3083, time=0.5259, value=0.2814}
          <at 0.9826 0.5>
         ]}
       }}
      CurveIndex=$System.Timeline001.NormalizedTime
     ))
    ScaleA=true
    ScaleRGB=true
   )
   /SetVariables(Particles.ColorCloudMasked:LinearColor=&MakeLinearColorFromVectorAndFloat(
     "Vector (RGB)"=&MakeVectorFromLinearColorRGB(Color=$Particles.Color)
     "Float (Alpha)"=&FloatFromCurve(
      FloatCurve=DataInterface() {propertyValues {Curve {keys [
          {arriveTangent=0.0328, interpMode=RCIM_Cubic, leaveTangent=0.0328, tangentMode=RCTM_User, time=0.1, value=0}
          {arriveTangent=-0.045, interpMode=RCIM_Cubic, leaveTangent=-0.045, tangentMode=RCTM_User, time=0.5, value=1}
         ]}}}
      CurveIndex=$Emitter.Age
     )
    )) {isSetParameters=true}
  ]
  rendererClasses [&NiagaraMeshRendererProperties]
  renderers [{rendererClass=&NiagaraMeshRendererProperties, rendererIndex=0}]
 }
 Rain {
  emitterUpdateScript [
   &EmitterState(Delay First Loop Only=true, Life Cycle Mode=::NewEnumerator1, Loop Delay=1, Loop Duration=4, UseLoopDelay=true)
   &SpawnRate(SpawnRate=&Lerp_Float(A=300, B=500, Alpha=$User.UmbrellaUp), Use Spawn Probability=false)
  ]
  particleSpawnScript [
   &InitializeParticle(
    Color=&RandomRangeLinearColor(Minimum=<rgba 1.6 1.8453 2 1>, Maximum=<rgba 0.8 1.5027 2 1>)
    Lifetime=&Multiply_Float(
     A=&RandomRangeFloat(Minimum=0.3, Maximum=0.35)
     B=&Lerp_Float(A=1, B=0.11, Alpha=$User.UmbrellaUp)
    )
    Mass Mode=::NewEnumerator3
    Sprite Size=&Multiply_Vector2D_ByFloat(
     Float=&Multiply_Float(
      A=&MakeCustomFloatFromBool(Bool=$User.bIsFrontEnd, True Float=1, False Float=1)
      B=&RandomRangeFloat(Minimum=8, Maximum=16)
     )
     Vector2D=<0.1 0.5>
    )
    Sprite Size Mode=::NewEnumerator4
    UsePositionOffset=false
    Write Material Random=true
   )
   &AddVelocity(Velocity=&RandomRangeVector(Minimum=<-20 -20 -600>, Maximum=<20 20 -400>))
   &ShapeLocation(
    Additional Quaternion Rotation=true
    Additional Yaw / Pitch / Roll=<0.25 0 0>
    MaskSpawnGroup=false
    Non Uniform Scale=<0.6 0.4 0.15>
    Override Seed=false
    Rotation Mode=::NewEnumerator3
    Shape Origin=$System.PositionCloud
    Sphere Radius=50
    UseOwnerScale=true
   )
  ]
  particleUpdateScript [
   &ParticleState()
   &GravityForce(Gravity=<0 0 -300>)
   &SolveForcesAndVelocity(Clamp Velocity=false, Limit Acceleration=false)
   &ScaleSpriteSizeBySpeed(Max Scale Factor=<1 3>, Velocity Threshold=2000)
   &ScaleColor(
    Scale Alpha=&FloatFromCurve(FloatCurve=DataInterface() {propertyValues {Curve {keys [
         <at 0 0 RCIM_Cubic>
         {arriveTangent=0.0435, interpMode=RCIM_Cubic, leaveTangent=0.0435, tangentMode=RCTM_User, time=0.15, value=1}
         {arriveTangent=0.0176, interpMode=RCIM_Cubic, leaveTangent=0.0176, tangentMode=RCTM_User, time=0.6, value=1}
         <at 1 0 RCIM_Cubic>
        ]}}})
    ScaleA=true
    ScaleRGB=true
   )
  ]
  rendererClasses [&NiagaraSpriteRendererProperties]
  renderers [{rendererClass=&NiagaraSpriteRendererProperties}]
 }
 Rain_SideAreas {
  emitterUpdateScript [
   &EmitterState(Delay First Loop Only=true, Life Cycle Mode=::NewEnumerator1, Loop Delay=1, Loop Duration=4, UseLoopDelay=true)
   &SpawnRate(SpawnRate=600, Use Spawn Probability=false)
  ]
  particleSpawnScript [
   &InitializeParticle(
    Color=&RandomRangeLinearColor(Minimum=<rgba 1.6 1.8453 2 1>, Maximum=<rgba 0.8 1.5027 2 1>)
    Lifetime=&RandomRangeFloat(Minimum=0.3, Maximum=0.35)
    Mass Mode=::NewEnumerator3
    Sprite Size=&Multiply_Vector2D_ByFloat(
     Float=&Multiply_Float(
      A=&MakeCustomFloatFromBool(Bool=$User.bIsFrontEnd, True Float=1, False Float=1)
      B=&RandomRangeFloat(Minimum=8, Maximum=16)
     )
     Vector2D=<0.1 0.5>
    )
    Sprite Size Mode=::NewEnumerator4
    UsePositionOffset=false
    Write Material Random=true
   )
   &PartitionParticles()
   &AddVelocity(Velocity=&RandomRangeVector(Minimum=<-20 -20 -600>, Maximum=<20 20 -400>))
   /SetVariables(Particles.Position=&AddVectorToPosition(
     Position=$System.PositionCloud
     Vector=&Multiply_Vector(
      A=&ReturnCameraVector(Camera Vectors=::NewEnumerator3, Coordinate Space=::Simulation)
      B=&VectorFromFloat(Value=&SelectFloatFromArray(Direct Array Index=$Particles.PartitionParticles.Partition, Float Selection Array=DataInterface() {propertyValues {FloatData [30,-25]}}))
     )
    )) {isSetParameters=true}
   &ShapeLocation(
    MaskSpawnGroup=false
    Non Uniform Scale=<1 1.3 0.5>
    Offset=<0 0 8>
    Offset Mode=::NewEnumerator0
    Override Seed=false
    Rotation Coordinate Space=::Simulation
    Shape Origin=$Particles.Position
    Sphere Radius=13
    UseOwnerScale=true
   )
   &ShapeLocation(
    MaskSpawnGroup=false
    Non Uniform Scale=<1 3 1>
    Override Seed=false
    Rotation Coordinate Space=::Simulation
    Rotation Mode=::NewEnumerator2
    Sphere Radius=20
    UseOwnerScale=true
    Yaw / Pitch / Roll=&Multiply_Vector(A=&ReturnCameraVector(Camera Vectors=::NewEnumerator3, Coordinate Space=::Simulation), B=<1 1 1>)
   ) {enabled=false, module=ShapeLocation001}
  ]
  particleUpdateScript [
   &ParticleState()
   &GravityForce(Gravity=<0 0 -300>)
   &SolveForcesAndVelocity(Clamp Velocity=false, Limit Acceleration=false)
   &ScaleSpriteSizeBySpeed(Max Scale Factor=<1 3>, Velocity Threshold=2000)
   &ScaleColor(
    Scale Alpha=&FloatFromCurve(FloatCurve=DataInterface() {propertyValues {Curve {keys [
         <at 0 0 RCIM_Cubic>
         {arriveTangent=0.0435, interpMode=RCIM_Cubic, leaveTangent=0.0435, tangentMode=RCTM_User, time=0.15, value=1}
         {arriveTangent=0.0176, interpMode=RCIM_Cubic, leaveTangent=0.0176, tangentMode=RCTM_User, time=0.6, value=1}
         <at 1 0 RCIM_Cubic>
        ]}}})
    ScaleA=true
    ScaleRGB=true
   )
  ]
  rendererClasses [&NiagaraSpriteRendererProperties]
  renderers [{rendererClass=&NiagaraSpriteRendererProperties}]
 }
 Rain_Sides {
  bEnabled=false
  emitterUpdateScript [
   &EmitterState(Delay First Loop Only=true, Life Cycle Mode=::NewEnumerator1, Loop Delay=1, Loop Duration=4, UseLoopDelay=true)
   &SpawnRate(SpawnRate=200, Use Spawn Probability=false)
  ]
  particleSpawnScript [
   &InitializeParticle(
    Color=&RandomRangeLinearColor(Minimum=<rgba 1.6 1.8453 2 1>, Maximum=<rgba 0.8 1.5027 2 1>)
    Lifetime=&RandomRangeFloat(Minimum=0.3, Maximum=0.35)
    Mass Mode=::NewEnumerator3
    Sprite Size=&Multiply_Vector2D_ByFloat(Float=&RandomRangeFloat(Minimum=8, Maximum=16), Vector2D=<0.1 0.5>)
    Sprite Size Mode=::NewEnumerator4
    UsePositionOffset=false
    Write Material Random=true
   )
   /Niagara/Modules/Spawn/Location/StaticMeshLocation(
    MaskSpawnGroup=false
    OffsetAlongNormal=false
    OverrideRandomSeed=false
    SamplePosOffset=false
    Static Mesh=DataInterface() {propertyValues {
      DefaultMesh=/Game/Animation/Game/MainPlayer/Emotes/Melancholy/VFX/Meshes/SM_Cloud_Squashed_LP
      InstanceIndex=-1
      LODIndex=0
      PreviewMesh=&SM_Cloud_Squashed_LP
      SoftSourceActor=null
      SourceMode=Default
      bAllowSamplingFromStreamingLODs=false
      bCaptureTransformsPerFrame=true
      bUsePhysicsBodyVelocity=false
      FilteredSockets []
      LODIndexUserParameter {parameter {name=None, typeDefHandle {registeredTypeIndex=108}, varData []}}
      MeshParameterBinding {parameter {name=None, typeDefHandle {registeredTypeIndex=1}, varData []}}
      SectionFilter {allowedMaterialSlots []}
     }}
   )
   &self:GetCloudMeshLocationModule(Particle Attribute Reader=DataInterface() {propertyValues {EmitterBinding {bindingMode=Other, emitterName=Cloud_Main}}}) {module=GetCloudMeshLocationModule}
   &RegionMask(Region Falloff=0, Region Radius=20, UseOffset=false, UseRotation=false) {enabled=false}
   /SetVariables(Particles.Lifetime=&MakeCustomFloatFromBool(
     Bool=$Particles.RegionMask.IsInsideMask
     True Float=&Multiply_Float(
      A=&RandomRangeFloat(Minimum=0.3, Maximum=0.35)
      B=&Lerp_Float(A=1, B=0.12, Alpha=$User.UmbrellaUp)
     )
     False Float=&RandomRangeFloat(Minimum=0.3, Maximum=0.35)
    )) {enabled=false, isSetParameters=true}
   &PartitionParticles() {enabled=false}
   &ShapeLocation(
    Additional Quaternion Rotation=true
    Additional Yaw / Pitch / Roll=<0.25 0 0>
    MaskSpawnGroup=false
    Non Uniform Scale=<1 1.3 0.5>
    Offset=&RotateVectorByQuaternion(
     Vector To Rotate=&SelectVectorFromArray(
      Direct Array Index=$Particles.PartitionParticles.Partition
      Vector Selection Array=DataInterface() {propertyValues {FloatData [<0 -30 10>, <0 26 10>]}}
     )
     Quaternion=$Engine.Owner.Rotation
    )
    Offset Mode=::NewEnumerator0
    Override Seed=false
    Rotation Mode=::NewEnumerator3
    Shape Origin=$System.PositionCloud
    Sphere Radius=16
    UseOwnerScale=true
   ) {enabled=false}
   &AddVelocity(Velocity=&RandomRangeVector(Minimum=<0 0 -600>, Maximum=<50 50 -400>)) {enabled=false}
   /SetVariables(Particles.Position=&Multiply_Vector(A=Linked() {linkedVariable {name=Particles.Position, type=NiagaraPosition}}, B=<1 1 0.4>)) {enabled=false, isSetParameters=true}
  ]
  particleUpdateScript [
   &ParticleState()
   &GravityForce(Gravity=<0 0 -300>) {enabled=false}
   &SolveForcesAndVelocity(Clamp Velocity=false, Limit Acceleration=false)
   &ScaleSpriteSizeBySpeed(Max Scale Factor=<1 3>, Velocity Threshold=2000)
   &ScaleColor(
    Scale Alpha=&FloatFromCurve(FloatCurve=DataInterface() {propertyValues {Curve {keys [
         <at 0 0 RCIM_Cubic>
         {arriveTangent=0.0435, interpMode=RCIM_Cubic, leaveTangent=0.0435, tangentMode=RCTM_User, time=0.15, value=1}
         {arriveTangent=0.0176, interpMode=RCIM_Cubic, leaveTangent=0.0176, tangentMode=RCTM_User, time=0.6, value=1}
         <at 1 0 RCIM_Cubic>
        ]}}})
    ScaleA=true
    ScaleRGB=true
   )
  ]
  rendererClasses [&NiagaraSpriteRendererProperties]
  renderers [
   {rendererClass=&NiagaraSpriteRendererProperties}
   {rendererClass=&NiagaraSpriteRendererProperties, rendererIndex=1}
  ]
 }
 ThickRainCurtain {
  emitterUpdateScript [
   &EmitterState(Delay First Loop Only=true, Life Cycle Mode=::NewEnumerator1, Loop Delay=1, Loop Duration=4, UseLoopDelay=true)
   &SpawnRate(Spawn Probability=0.94, SpawnRate=&Lerp_Float(A=10, B=16, Alpha=$User.UmbrellaUp), Use Spawn Probability=false)
  ]
  particleSpawnScript [
   &InitializeParticle(
    Color Maximum=<rgba 0.12 0.1551 0.2 1>
    Color Minimum=<rgba 0.045 0.0691 0.1 1>
    Color Mode=::NewEnumerator2
    Lifetime=&Multiply_Float(
     A=&RandomRangeFloat(Minimum=0.6, Maximum=0.8)
     B=&Lerp_Float(A=1.2, B=0.6, Alpha=$User.UmbrellaUp)
    )
    Mesh Scale Mode=::NewEnumerator3
    Mesh Uniform Scale=0.5
    Position=&AddVectorToPosition(
     Position=$Engine.Emitter.SimulationPosition
     Vector=&VectorFromFloat(Value=&FloatFromCurve(
       FloatCurve=DataInterface() {propertyValues {Curve {keys [<at 0 -1 RCIM_Cubic>,<at 0.25 1 RCIM_Cubic>,<at 0.5 -1 RCIM_Cubic>,<at 0.75 1 RCIM_Cubic>,<at 1 -1 RCIM_Cubic>]}}}
       CurveIndex=$System.NormalizedLoopAge
       Scale Curve=4
      ))
    )
    Position Mode=::NewEnumerator1
    Position Offset=&Lerp_Vector(A=<0 0 10>, B=<0 0 20>, Alpha=$User.UmbrellaUp)
    Sprite Size=&MakeVector2D(
     X=&RandomRangeFloat(Minimum=85, Maximum=100)
     Y=&Multiply_Float(
      A=&RandomRangeFloat(Minimum=140, Maximum=160)
      B=&Lerp_Float(A=1, B=0.36, Alpha=$User.UmbrellaUp)
     )
    )
    Sprite Size Mode=::NewEnumerator4
    UsePositionOffset=true
    Write Material Random=true
   )
   &AddVelocity(Velocity=&RandomRangeVector(Minimum=<0 0 -230>, Maximum=<0 0 -280>))
   &CameraOffset(Camera Offset Amount=12)
   /SetVariables(Particles.RainWavesRandomOpacity:NiagaraFloat=&Multiply_Float(
     A=&MakeCustomFloatFromBool(
      Bool=$User.bIsFrontEnd
      True Float=&RandomRangeFloat(Minimum=0.9, Maximum=1)
      False Float=&RandomRangeFloat(Minimum=0.75, Maximum=0.9)
     )
     B=&FN_Select_Float_from_Time_Of_Day(Morning=0.7, Day=1, Evening=1, Night=0.3)
    )) {isSetParameters=true}
  ]
  particleUpdateScript [
   &ParticleState()
   &ScaleSpriteSize(
    Scale Factor=&MakeVector2D(
     X=&FloatFromCurve(FloatCurve=DataInterface() {propertyValues {Curve {keys [
          {interpMode=RCIM_Cubic, tangentMode=RCTM_User, time=0, value=0.3}
          {arriveTangent=-0.0137, interpMode=RCIM_Cubic, leaveTangent=-0.0137, tangentMode=RCTM_User, time=0.3, value=1}
          {arriveTangent=-0.283, interpMode=RCIM_Cubic, leaveTangent=-0.283, tangentMode=RCTM_User, time=1, value=0.8}
         ]}}})
     Y=&FloatFromCurve(FloatCurve=DataInterface() {propertyValues {Curve {keys [
          {arriveTangent=1.0683, interpMode=RCIM_Cubic, leaveTangent=1.0683, tangentMode=RCTM_User, time=0, value=0.2}
          {arriveTangent=-0.0018, interpMode=RCIM_Cubic, leaveTangent=-0.0018, tangentMode=RCTM_User, time=0.3, value=1}
          {arriveTangent=-0.283, interpMode=RCIM_Cubic, leaveTangent=-0.283, tangentMode=RCTM_User, time=1, value=0.8}
         ]}}})
    )
    Scale Sprite Size Mode=::NewEnumerator2
   )
   &DynamicMaterialParameters(
    Index 0 Param 1=&FloatFromCurve(
     FloatCurve=DataInterface() {propertyValues {Curve {keys [
         <at 0 0 RCIM_Cubic>
         {arriveTangent=0.0051, interpMode=RCIM_Cubic, leaveTangent=0.0051, tangentMode=RCTM_User, time=0.05, value=1}
         <at 1 0.7 RCIM_Cubic>
        ]}}}
     CurveIndex=$Particles.Age
     Scale Curve=$Particles.RainWavesRandomOpacity
    )
    Module.Param0WriteEnabled=true
    Module.Param1WriteEnabled=false
    Module.Param2WriteEnabled=false
    Module.Param3WriteEnabled=false
   )
   &GravityForce() {enabled=false}
   &SolveForcesAndVelocity(Clamp Velocity=false, Limit Acceleration=false)
   &ScaleColor(
    Scale Alpha=&FloatFromCurve(FloatCurve=DataInterface() {propertyValues {Curve {keys [{arriveTangent=0.0176, interpMode=RCIM_Cubic, leaveTangent=0.0176, tangentMode=RCTM_User, time=0.5, value=1}, <at 1 0 RCIM_Cubic>]}}})
    Scale RGB=&MakeVectorFromLinearColorRGB(Color=&FNTimeOfDayColor(Night Color=<rgba 0.6 0.6667 1 1>, Morning Color=<rgba 0.9152 0.8 1 1>, Day Color=<rgba 1 1 1 1>, Evening Color=<rgba 1 0.6744 0.6 1>))
    ScaleA=true
    ScaleRGB=true
   )
  ]
  rendererClasses [&NiagaraSpriteRendererProperties]
  renderers [{rendererClass=&NiagaraSpriteRendererProperties}]
 }
 Water_Splashes {
  emitterSpawnScript [/SetVariables(Emitter.PositionRainImpacts:NiagaraPosition=&AddVectorToPosition(Position=$Engine.Emitter.SimulationPosition, Vector=<5 2 -23>)) {isSetParameters=true}]
  emitterUpdateScript [
   &EmitterState()
   &SpawnRate(SpawnRate=&Multiply_Float(A=150, B=$User.UmbrellaUp), Use Spawn Probability=false)
  ]
  particleSpawnScript [
   &InitializeParticle(
    Color=&FNTimeOfDayColor(
     Night Color=<rgba 0.6 0.6 0.6 0.7>
     Morning Color=<rgba 0.65 0.3522 0.13 1>
     Day Color=<rgba 0.6 0.6 0.6 1>
     Evening Color=<rgba 0.8 0.5353 0.3378 1>
    )
    Lifetime Max=0.14
    Lifetime Min=0.08
    Lifetime Mode=::NewEnumerator1
    Sprite Rotation Mode=::NewEnumerator2
    Sprite Size Mode=::NewEnumerator1
    Uniform Sprite Size Max=9
    Uniform Sprite Size Min=3
    UsePositionOffset=false
    Write Material Random=true
   )
   &ShapeLocation(
    Hemisphere Distribution=<1 0.4>
    MaskSpawnGroup=false
    Non Uniform Scale=<1 1 0.5>
    Offset Mode=::NewEnumerator0
    Override Seed=false
    Shape Origin=$Emitter.PositionRainImpacts
    Sphere Radius=17
    Sphere Surface Distribution=1
    UseOwnerScale=true
   )
   &AddVelocity(
    Module.Use Falloff Distance=false
    Origin Offset=<0 0 -40>
    Override Seed=false
    Velocity Mode=::NewEnumerator1
    Velocity Origin=$Emitter.PositionRainImpacts
    Velocity Speed=&RandomRangeFloat(Minimum=30, Maximum=80)
   )
   &SubUVAnimation(
    "Sprite Renderer "=DataInterface() {propertyValues {SpriteRenderer=&self:Water_Splashes_0.NiagaraSpriteRendererProperties_0}}
    SubUV Animation Mode:ENiagara_SubUVLookupModeV2=::NewEnumerator2
    UseEndFrame=false
    UseStartFrame=false
   )
   &CameraOffset(Camera Offset Amount=5)
  ]
  particleUpdateScript [
   &ParticleState()
   &SpriteRotationRate(Rotation Rate=&FloatFromCurve(
     FloatCurve=DataInterface() {propertyValues {Curve {keys [<at 0 1 RCIM_Cubic>,<at 0.7 0 RCIM_Cubic>]}}}
     Scale Curve=&RandomRangeFloat(Minimum=-300, Maximum=300)
    ))
   &GravityForce()
   &SolveForcesAndVelocity(Clamp Velocity=false, Limit Acceleration=false)
   &ScaleColor(
    Scale Alpha=&FloatFromCurve(FloatCurve=DataInterface() {propertyValues {Curve {keys [
         {interpMode=RCIM_Cubic, tangentMode=RCTM_User, time=0, value=0}
         {interpMode=RCIM_Cubic, tangentMode=RCTM_User, time=0.05, value=1}
         {interpMode=RCIM_Cubic, tangentMode=RCTM_User, time=0.5, value=1}
         {interpMode=RCIM_Cubic, tangentMode=RCTM_User, time=1, value=0}
        ]}}})
    ScaleA=true
    ScaleRGB=true
   )
   &DynamicMaterialParameters(
    Index 0 Param 1=&FloatFromCurve(FloatCurve=DataInterface() {propertyValues {Curve {keys [
         {interpMode=RCIM_Cubic, tangentMode=RCTM_User, time=0.3, value=1}
         {arriveTangent=-1.9654, interpMode=RCIM_Cubic, leaveTangent=-1.9654, tangentMode=RCTM_User, time=1, value=0}
        ]}}})
    Module.Param0WriteEnabled=true
    Module.Param1WriteEnabled=false
    Module.Param2WriteEnabled=false
    Module.Param3WriteEnabled=false
   ) {enabled=false}
  ]
  rendererClasses [&NiagaraSpriteRendererProperties]
  renderers [{rendererClass=&NiagaraSpriteRendererProperties}]
 }
}
summary {
 systemName=NS_Emote_Melancholy
 emitters [
  {emitterName=Cloud_Main, rendererClasses [&NiagaraMeshRendererProperties]}
  {emitterName=Rain, rendererClasses [&NiagaraSpriteRendererProperties]}
  {emitterName=ThickRainCurtain, rendererClasses [&NiagaraSpriteRendererProperties]}
  {emitterName=Water_Splashes, rendererClasses [&NiagaraSpriteRendererProperties]}
  {bEnabled=false, emitterName=Rain_Sides, rendererClasses [&NiagaraSpriteRendererProperties]}
  {emitterName=CloudPuffsExtra, rendererClasses [&NiagaraSpriteRendererProperties]}
  {emitterName=Rain_SideAreas, rendererClasses [&NiagaraSpriteRendererProperties]}
 ]
 userVariables [
  User.bIsFrontEnd {type=NiagaraBool}
  User.bIsFrontEndPreview {type=NiagaraBool}
  User.LightningStrikeActive {type=NiagaraFloat, defaultValue=0, description=""}
  User.NormalizedNotifyProgress {type=NiagaraFloat, defaultValue=0, description=""}
  User.SkeletalMesh {
   type=NiagaraDataInterfaceSkeletalMesh
   defaultValue {
    struct=NiagaraExt_VariableValue_DataInterface
    value {dataInterface=&self:NiagaraDataInterfaceSkeletalMesh_0}
   }
  }
  User.UmbrellaUp {type=NiagaraFloat, defaultValue=0, description=""}
 ]
}
