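# Niagara system for the Melancholy emote's lightning strike: a brief purple-blue flash with thunder audio.
# FakeLights fires three early bursts of glow sprites, using partition-indexed offsets and sizes, and plays the thunder cue on particle spawn.
# Lightning then double-bursts randomly oriented meshes with lights and flash sprites, while a delayed Ring adds an expanding, camera-offset halo.
# Color curves spike into HDR and decay; no module shown reads the user parameters.
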NiagaraSystem /BRCosmetics/Animation/Game/MainPlayer/Emotes/Melancholy/VFX/NS_Emote_Melancholy_LightningStrike
dependencies {
 usedDataInterfaces [
  &NiagaraDataInterfaceColorCurve
  &NiagaraDataInterfaceCurve
  &NiagaraDataInterfaceArrayFloat3
  &NiagaraDataInterfaceArrayFloat
  &NiagaraDataInterfaceAudioPlayer
 ]
 usedDynamicInputs []
 usedModules [&SystemState]
}
emitters {
 FakeLights {
  emitterUpdateScript [
   &EmitterState(Life Cycle Mode=::NewEnumerator1, Loop Behavior=::NewEnumerator1, Loop Duration=1, UseLoopDelay=false)
   &SpawnBurst_Instantaneous(Use Spawn Probability=false, UseLoopCountLimit=false)
   &SpawnBurst_Instantaneous(Spawn Time=0.05, Use Spawn Probability=false, UseLoopCountLimit=false) {module=SpawnBurst_Instantaneous001}
   &SpawnBurst_Instantaneous(Spawn Time=0.25, Use Spawn Probability=false, UseLoopCountLimit=false) {module=SpawnBurst_Instantaneous002}
  ]
  particleSpawnScript [
   &PartitionParticles(Number Of Partitions=3)
   &InitializeParticle(
    Color=<rgba 52.7 3.6 200 1>
    Lifetime=0.45
    Position Offset=&SelectVectorFromArray(
     Direct Array Index=$Particles.PartitionParticles.Partition
     Vector Selection Array=DataInterface() {propertyValues {FloatData [<2 -32 7>, <2 10 8>, <0 -5 -7>]}}
    )
    Sprite Size Mode=::NewEnumerator3
    Uniform Sprite Size=&SelectFloatFromArray(Direct Array Index=$Particles.PartitionParticles.Partition, Float Selection Array=DataInterface() {propertyValues {FloatData [40,65,70]}})
    UsePositionOffset=true
    Write Material Random=true
   )
   /Niagara/Modules/Audio/PlayAudio(Audio Player=DataInterface() {propertyValues {
      Attenuation=null
      Concurrency=null
      MaxPlaysPerTick=10
      SoundToPlay=/Game/Athena/Sounds/Emotes/Melancholy/SC_EmoteFoley_Melancholy_Thunder
      bAllowLoopingOneShotSounds=false
      bLimitPlaysPerTick=true
      bOnlyActiveDuringGameplay=false
      bStopWhenComponentIsDestroyed=true
      ConfigurationUserParameter {parameter {name=None, typeDefHandle {registeredTypeIndex=1}, varData []}}
      ParameterNames []
     }})
  ]
  particleUpdateScript [
   &ParticleState()
   &ScaleColor(
    Scale Alpha=&FloatFromCurve(FloatCurve=DataInterface() {propertyValues {Curve {keys [<at 0 0>,<at 0.1 1>,<at 0.9 1>,<at 1 0>]}}})
    Scale RGB=&MakeVectorFromLinearColorRGB(Color=&ColorFromCurve(ColorCurve=DataInterface() {propertyValues {
        AlphaCurve {keys []}
        BlueCurve {keys [<at 0.0209 10 RCIM_Cubic>,<at 0.1049 1>,<at 1 0.5 RCIM_Cubic>]}
        GreenCurve {keys [<at 0.0209 10 RCIM_Cubic>,<at 0.1049 1>,<at 1 0.5 RCIM_Cubic>]}
        RedCurve {keys [<at 0.0209 10 RCIM_Cubic>,<at 0.1049 1>,<at 1 0.5 RCIM_Cubic>]}
       }}))
    ScaleA=true
    ScaleRGB=true
   )
  ]
  rendererClasses [&NiagaraSpriteRendererProperties]
  renderers [
   {rendererClass=&NiagaraSpriteRendererProperties}
   {rendererClass=&NiagaraSpriteRendererProperties, rendererIndex=1}
  ]
 }
 Lightning {
  emitterUpdateScript [
   &EmitterState(Life Cycle Mode=::NewEnumerator1, Loop Behavior=::NewEnumerator1, Loop Duration=1, UseLoopDelay=false)
   &SpawnBurst_Instantaneous(Spawn Time=0.33, Use Spawn Probability=false, UseLoopCountLimit=false)
   &SpawnBurst_Instantaneous(Spawn Time=0.38, Use Spawn Probability=false, UseLoopCountLimit=false) {module=SpawnBurst_Instantaneous001}
  ]
  particleSpawnScript [
   &InitializeParticle(
    Color=<rgba 0.1475 0.07 1 0.3>
    Lifetime=0.15
    Mesh Scale=&Multiply_VectorByFloat(Vector=<1 1 0.7>, Float=1)
    Mesh Scale Mode=::NewEnumerator4
    Position Offset=<0 0 5>
    Sprite Size Mode=::NewEnumerator3
    Uniform Sprite Size=80
    UsePositionOffset=true
    Write Material Random=true
   )
   /SetVariables(Particles.PositionFlash:NiagaraPosition=&AddVectorToPosition(Position=$Particles.Position, Vector=<0 0 -15>)) {isSetParameters=true}
   &InitialMeshOrientation(
    Mesh Orientation Mode=::NewEnumerator2
    Rotation=&RandomRangeVector(Minimum=<0 0 -1>, Maximum=<0 0 1>)
    Use Rotation Vector=true
   )
  ]
  particleUpdateScript [
   &ParticleState()
   &ScaleColor(
    Scale Alpha=&FloatFromCurve(FloatCurve=DataInterface() {propertyValues {Curve {keys [{arriveTangent=-0.0462, interpMode=RCIM_Cubic, leaveTangent=-0.0462, tangentMode=RCTM_User, time=0, value=1}, <at 0.6 0 RCIM_Cubic>]}}})
    Scale RGB=&MakeVectorFromLinearColorRGB(Color=&ColorFromCurve(ColorCurve=DataInterface() {propertyValues {
        AlphaCurve {keys []}
        BlueCurve {keys [<at 0.0919 10 RCIM_Cubic>,<at 0.3389 0.1 RCIM_Cubic>]}
        GreenCurve {keys [<at 0.0919 10 RCIM_Cubic>,<at 0.3389 0.1 RCIM_Cubic>]}
        RedCurve {keys [<at 0.0919 10 RCIM_Cubic>,<at 0.3389 0.1 RCIM_Cubic>]}
       }}))
    ScaleA=true
    ScaleRGB=true
   )
   /SetVariables(Particles.Color_LightningGlow:LinearColor=&MakeLinearColorFromVectorAndFloat(
     "Vector (RGB)"=&MakeVectorFromLinearColorRGB(Color=<rgba 0.1567 0.08 1 1>)
     "Float (Alpha)"=&FloatFromCurve(FloatCurve=DataInterface() {propertyValues {Curve {keys [{arriveTangent=-0.0462, interpMode=RCIM_Cubic, leaveTangent=-0.0462, tangentMode=RCTM_User, time=0, value=1}, <at 0.8 0 RCIM_Cubic>]}}})
    )) {isSetParameters=true}
  ]
  rendererClasses [
   &NiagaraMeshRendererProperties
   &NiagaraLightRendererProperties
   &NiagaraSpriteRendererProperties
  ]
  renderers [
   {rendererClass=&NiagaraMeshRendererProperties, rendererIndex=0}
   {rendererClass=&NiagaraLightRendererProperties, rendererIndex=1}
   {rendererClass=&NiagaraSpriteRendererProperties, rendererIndex=2}
  ]
 }
 Ring {
  emitterUpdateScript [
   &EmitterState(Life Cycle Mode=::NewEnumerator1, Loop Behavior=::NewEnumerator1, Loop Delay=0.32, Loop Duration=1, UseLoopDelay=true)
   &SpawnBurst_Instantaneous(Spawn Count=1, Use Spawn Probability=false)
  ]
  particleSpawnScript [
   &InitializeParticle(
    Color=<rgba 0.4725 0.2921 0.5 1>
    Lifetime=0.12
    Position Offset=<0 0 -18>
    Sprite Size Mode=::NewEnumerator3
    Uniform Sprite Size=40
    UsePositionOffset=true
    Write Material Random=true
   )
   &CameraOffset(Camera Offset Amount=60)
   /SetVariables(Particles.SpriteStarSize:Vector2f=&Multiply_Vector2D_ByFloat(Float=20, Vector2D=<3 1>)) {isSetParameters=true}
  ]
  particleUpdateScript [
   &ParticleState()
   &ScaleColor(
    Scale Alpha=&FloatFromCurve(FloatCurve=DataInterface() {propertyValues {Curve {keys [<at 0 1 RCIM_Cubic>,<at 1 0 RCIM_Cubic>]}}})
    Scale RGB=&MakeVectorFromLinearColorRGB(Color=&ColorFromCurve(ColorCurve=DataInterface() {propertyValues {
        AlphaCurve {keys []}
        BlueCurve {keys [<at 0.1434 60 RCIM_Cubic>,<at 0.7038 2 RCIM_Cubic>]}
        GreenCurve {keys [<at 0.1434 60 RCIM_Cubic>,<at 0.7038 0.1301 RCIM_Cubic>]}
        RedCurve {keys [<at 0.1434 60 RCIM_Cubic>,<at 0.7038 0.1134 RCIM_Cubic>]}
       }}))
    ScaleA=true
    ScaleRGB=true
   )
   /SetVariables(Particles.ColorHalo:LinearColor=&MakeLinearColorFromVectorAndFloat(
     "Vector (RGB)"=&MakeVectorFromLinearColorRGB(Color=&ColorFromCurve(ColorCurve=DataInterface() {propertyValues {
         AlphaCurve {keys []}
         BlueCurve {keys [<at 0 10 RCIM_Cubic>,<at 0.7127 0.5 RCIM_Cubic>]}
         GreenCurve {keys [<at 0 0.5 RCIM_Cubic>,<at 0.7127 0.5 RCIM_Cubic>]}
         RedCurve {keys [<at 0 4.1623 RCIM_Cubic>,<at 0.7127 0.5 RCIM_Cubic>]}
        }}))
     "Float (Alpha)"=&FloatFromCurve(FloatCurve=DataInterface() {propertyValues {Curve {keys [<at 0 1 RCIM_Cubic>,<at 1 0 RCIM_Cubic>]}}})
    )) {isSetParameters=true}
   &ScaleSpriteSize(Scale Factor=&Vector2DFromFloat(Value=&FloatFromCurve(FloatCurve=DataInterface() {propertyValues {Curve {keys [{arriveTangent=1.2868, interpMode=RCIM_Cubic, leaveTangent=1.2868, tangentMode=RCTM_User, time=0, value=0.5}, <at 0.9 1>]}}})))
  ]
  rendererClasses [&NiagaraSpriteRendererProperties]
  renderers [
   {rendererClass=&NiagaraSpriteRendererProperties}
   {rendererClass=&NiagaraSpriteRendererProperties, rendererIndex=1}
  ]
 }
}
summary {
 systemName=NS_Emote_Melancholy_LightningStrike
 emitters [
  {
   emitterName=Lightning
   rendererClasses [
    &NiagaraMeshRendererProperties
    &NiagaraLightRendererProperties
    &NiagaraSpriteRendererProperties
   ]
  }
  {emitterName=FakeLights, rendererClasses [&NiagaraSpriteRendererProperties]}
  {emitterName=Ring, rendererClasses [&NiagaraSpriteRendererProperties]}
 ]
 userVariables [
  User.bIsFrontEnd {type=NiagaraBool}
  User.bIsFrontEndPreview {type=NiagaraBool}
  User.LightningStrikeActive {type=NiagaraFloat, defaultValue=0, description=""}
  User.NormalizedNotifyProgress {type=NiagaraFloat, defaultValue=0, description=""}
  User.SkeletalMesh {
   type=NiagaraDataInterfaceSkeletalMesh
   defaultValue {
    struct=NiagaraExt_VariableValue_DataInterface
    value {dataInterface=&self:NiagaraDataInterfaceSkeletalMesh_0}
   }
  }
  User.UmbrellaUp {type=NiagaraFloat, defaultValue=0, description=""}
 ]
}
