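# Niagara system for the Melancholy emote that dissolves the umbrella gadget in a short puff of purple-glowing ash and smoke.
# Ashes and AshesFalling spawn sprites from filtered gadget skeletal bones, one swirled by curl noise, the other pulled down by gravity.
# Smoke and UmbrellaColoring position themselves from a system-level skeletal mesh sample, adding expanding puffs and a growing sprite flash.
# Color curves fade the bright purple to black or neutral tint.
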NiagaraSystem /BRCosmetics/Animation/Game/MainPlayer/Emotes/Melancholy/VFX/NS_Emote_Melancholy_UmbrellaDissolve
dependencies {
 usedDataInterfaces [
  &NiagaraDataInterfaceCurve
  &NiagaraDataInterfaceSpriteRendererInfo
  &NiagaraDataInterfaceColorCurve
 ]
 usedDynamicInputs [&MakeFloatFromVector4]
 usedModules [&SampleSkeletalMesh, &SystemState]
}
emitters {
 Ashes {
  emitterUpdateScript [
   &EmitterState(Life Cycle Mode=::NewEnumerator1, Loop Behavior=::NewEnumerator1, Loop Delay=0.05, Loop Duration=0.23, UseLoopDelay=true)
   &SpawnRate(SpawnRate=400, Use Spawn Probability=false)
  ]
  particleSpawnScript [
   &InitializeParticle(
    Lifetime Max=0.8
    Lifetime Min=0.4
    Lifetime Mode=::NewEnumerator1
    Mass Min=0.2
    Mass Mode=::NewEnumerator3
    Position Offset=&RandomRangeVector(Minimum=<-2 -20 -20>, Maximum=<20 20 20>)
    Sprite Rotation Mode=::NewEnumerator2
    Sprite Size Mode=::NewEnumerator1
    Uniform Sprite Size Max=5
    Uniform Sprite Size Min=2
    UsePositionOffset=false
    Write Material Random=true
   )
   &SkeletalMeshLocation(
    Bone Sampling Mode=::NewEnumerator2
    MaskSpawnGroup=false
    OverrideRandomSeed=false
    Sampled Position Offset=&RandomRangeVector(Minimum=<-5 -5 -5>, Maximum=<5 5 5>)
    Sampled Position Offset Coordinate Space:ENiagaraSkelMeshTransforms=::NewEnumerator0
    Skeletal Mesh=DataInterface() {propertyValues {
      DefaultMesh=null
      ExcludeBoneName=root
      OptionalComponentClass=null
      PreviewMesh=/BRCosmetics/Animation/Props/Gadgets_Melancholy/Meshes/Melancholy_Gadget
      SkinningMode=SkinOnTheFly
      SoftSourceActor=null
      SourceMode=Default
      UvSetIndex=0
      WholeMeshLOD=-1
      bExcludeBone=false
      bReadDeformedGeometry=true
      bRequireCurrentFrameData=true
      ComponentTags []
      FilteredBones [root,offset,main,JNT_FX_Cloud]
      FilteredSockets []
      MeshUserParameter {parameter {name=None, typeDefHandle {registeredTypeIndex=1}, varData []}}
      SamplingRegions []
     }}
   )
   &CameraOffset(Camera Offset Amount=25)
   &CurlNoiseForce(
    Noise Frequency=&RandomRangeFloat(Minimum=50, Maximum=100)
    Noise Strength=&RandomRangeFloat(Minimum=5, Maximum=20)
    Pan Noise=false
   )
   &ApplyInitialForces()
  ]
  particleUpdateScript [
   &ParticleState()
   &ScaleSpriteSize(Uniform Curve Sprite Scale=DataInterface() {propertyValues {Curve {keys [
        {interpMode=RCIM_Cubic, tangentMode=RCTM_User, time=0.1, value=0.1}
        {interpMode=RCIM_Cubic, tangentMode=RCTM_User, time=0.3, value=1}
        <at 0.9 0.2>
       ]}}})
   &SubUVAnimation(
    "Sprite Renderer "=DataInterface() {propertyValues {SpriteRenderer=&self:Ashes_0.NiagaraSpriteRendererProperties_0}}
    UseEndFrame=false
    UseStartFrame=false
   )
   &ScaleColor(
    Scale Alpha=&FloatFromCurve(FloatCurve=DataInterface() {propertyValues {Curve {keys [{interpMode=RCIM_Cubic, tangentMode=RCTM_User, time=0.5, value=1}, {interpMode=RCIM_Cubic, tangentMode=RCTM_User, time=1, value=0}]}}})
    Scale RGB=&MakeVectorFromLinearColorRGB(Color=&ColorFromCurve(ColorCurve=DataInterface() {propertyValues {
        AlphaCurve {keys []}
        BlueCurve {keys [<at 0.1785 500 RCIM_Cubic>,<at 0.5325 3>,<at 1 0 RCIM_Cubic>]}
        GreenCurve {keys [<at 0.1785 0 RCIM_Cubic>,<at 0.5325 0.06>,<at 1 0 RCIM_Cubic>]}
        RedCurve {keys [<at 0.1785 41.667 RCIM_Cubic>,<at 0.5325 0.795>,<at 1 0 RCIM_Cubic>]}
       }}))
    ScaleA=true
    ScaleRGB=true
   )
   &DynamicMaterialParameters(
    Index 0 Param 1=&RandomRangeFloat(Minimum=0, Maximum=1)
    Index 0 Param 2=&RandomRangeFloat(Minimum=0, Maximum=1)
    Module.Param0WriteEnabled=true
    Module.Param1WriteEnabled=true
    Module.Param2WriteEnabled=true
    Module.Param3WriteEnabled=true
   )
   &SolveForcesAndVelocity(Clamp Velocity=false, Limit Acceleration=false)
  ]
  rendererClasses [&NiagaraSpriteRendererProperties]
  renderers [{rendererClass=&NiagaraSpriteRendererProperties}]
 }
 AshesFalling {
  emitterUpdateScript [
   &EmitterState(Life Cycle Mode=::NewEnumerator1, Loop Behavior=::NewEnumerator1, Loop Delay=0.07, Loop Duration=0.25, UseLoopDelay=true)
   &SpawnRate(SpawnRate=400, Use Spawn Probability=false)
  ]
  particleSpawnScript [
   &InitializeParticle(
    Color=<rgba 0 0 0 4>
    Lifetime Max=0.8
    Lifetime Min=0.3
    Lifetime Mode=::NewEnumerator1
    Mass Max=3
    Mass Min=0.3
    Mass Mode=::NewEnumerator3
    Sprite Rotation Mode=::NewEnumerator2
    Sprite Size Mode=::NewEnumerator1
    Uniform Sprite Size Max=4
    Uniform Sprite Size Min=2
    UsePositionOffset=false
    Write Material Random=true
   )
   &SkeletalMeshLocation(
    Bone Sampling Mode=::NewEnumerator2
    MaskSpawnGroup=false
    OverrideRandomSeed=false
    Sampled Position Offset=&RandomRangeVector(Minimum=<-5 -5 -5>, Maximum=<5 5 5>)
    Sampled Position Offset Coordinate Space:ENiagaraSkelMeshTransforms=::NewEnumerator0
    Skeletal Mesh=DataInterface() {propertyValues {
      DefaultMesh=null
      ExcludeBoneName=root
      OptionalComponentClass=null
      PreviewMesh=&Melancholy_Gadget
      SkinningMode=SkinOnTheFly
      SoftSourceActor=null
      SourceMode=Default
      UvSetIndex=0
      WholeMeshLOD=-1
      bExcludeBone=false
      bReadDeformedGeometry=true
      bRequireCurrentFrameData=true
      ComponentTags []
      FilteredBones [root,offset,main,JNT_FX_Cloud]
      FilteredSockets []
      MeshUserParameter {parameter {name=None, typeDefHandle {registeredTypeIndex=1}, varData []}}
      SamplingRegions []
     }}
   )
   &CameraOffset(Camera Offset Amount=25)
   &GravityForce() {enabled=false}
   &ApplyInitialForces()
  ]
  particleUpdateScript [
   &ParticleState()
   &ScaleSpriteSize(Uniform Curve Sprite Scale=DataInterface() {propertyValues {Curve {keys [{interpMode=RCIM_Cubic, tangentMode=RCTM_User, time=0.1, value=1}, <at 1 0.1>]}}})
   &SubUVAnimation(
    "Sprite Renderer "=DataInterface() {propertyValues {SpriteRenderer=&self:AshesFalling_0.NiagaraSpriteRendererProperties_0}}
    UseEndFrame=false
    UseStartFrame=false
   )
   &GravityForce(Gravity=<0 0 -600>) {module=GravityForce001}
   &SolveForcesAndVelocity(Clamp Velocity=false, Limit Acceleration=false)
   &ScaleSpriteSizeBySpeed(Max Scale Factor=<1 3>, Velocity Threshold=400) {enabled=false}
   &ScaleColor(
    Scale Alpha=&FloatFromCurve(FloatCurve=DataInterface() {propertyValues {Curve {keys [{interpMode=RCIM_Cubic, tangentMode=RCTM_User, time=0.7, value=1}, {interpMode=RCIM_Cubic, tangentMode=RCTM_User, time=1, value=0}]}}})
    ScaleA=true
    ScaleRGB=false
   )
   &DynamicMaterialParameters(
    Index 0 Param 1=&RandomRangeFloat(Minimum=0, Maximum=1)
    Index 0 Param 2=&RandomRangeFloat(Minimum=0, Maximum=1)
    Module.Param0WriteEnabled=true
    Module.Param1WriteEnabled=true
    Module.Param2WriteEnabled=true
    Module.Param3WriteEnabled=true
   )
  ]
  rendererClasses [&NiagaraSpriteRendererProperties]
  renderers [{rendererClass=&NiagaraSpriteRendererProperties}]
 }
 Smoke {
  emitterUpdateScript [
   &EmitterState(Life Cycle Mode=::NewEnumerator1, Loop Behavior=::NewEnumerator1, Loop Delay=0.06, Loop Duration=0.4, UseLoopDelay=true)
   &SpawnBurst_Instantaneous(Spawn Count=3, Spawn Time=0.2, Use Spawn Probability=false, UseLoopCountLimit=false)
   &SpawnRate(SpawnRate=12, Use Spawn Probability=false)
  ]
  particleSpawnScript [
   /SetVariables(Particles.RandomFloat=&UniformRangedFloat(Minimum=0, Maximum=1)) {isSetParameters=true}
   &InitializeParticle(
    Color=<rgba 0.0573 0.0495 0.0489 1>
    Lifetime=&RandomRangeFloat(Minimum=0.5, Maximum=0.7)
    Sprite Rotation Mode=::NewEnumerator2
    Sprite Size Mode=::NewEnumerator1
    Uniform Sprite Size Max=17
    Uniform Sprite Size Min=10
    UsePositionOffset=false
    Write Material Random=true
   ) {module=InitializeParticle001}
   &DynamicMaterialParameters(
    Index 0 Param 4=0.25
    Index 2 Param 2=&RandomRangeFloat(Minimum=0, Maximum=1)
    Module.Param0WriteEnabled=false
    Module.Param0WriteEnabled2=false
    Module.Param1WriteEnabled=false
    Module.Param1WriteEnabled2=true
    Module.Param2WriteEnabled=false
    Module.Param2WriteEnabled2=false
    Module.Param3WriteEnabled=true
    Module.Param3WriteEnabled2=false
    Write Parameter Index 2=true
   )
   &AddVelocity(
    Rotation Coordinate Space=::World
    Velocity=$System.Direction
    Velocity Speed Scale=&Multiply_Float(A=$System.Blowing, B=100)
   ) {enabled=false, module=AddVelocity001}
   &ShapeLocation(
    MaskSpawnGroup=false
    Non Uniform Scale=<1 1 0.6>
    Offset=<0 0 -10>
    Offset Mode=::NewEnumerator0
    Override Seed=false
    Shape Origin=$System.SampleSkeletalMesh.SampledPosition
    Sphere Radius=10
    UseOwnerScale=true
   )
   &AddVelocity(Module.Use Falloff Distance=false, Origin Offset=<0 0 -2>, Override Seed=false, Velocity Mode=::NewEnumerator1, Velocity Speed=40)
  ]
  particleUpdateScript [
   &ParticleState()
   &Drag(Drag=4, Use Linear Drag=true, Use Rotational Drag=false)
   &WindForce(Wind Speed=<0 0 1>, Wind Speed Scale=&RandomRangeFloat(Minimum=350, Maximum=420)) {enabled=false}
   &DynamicMaterialParameters(
    Index 0 Param 1=&Add_Float(
     A=$Particles.RandomFloat
     B=&FloatFromCurve(FloatCurve=DataInterface() {propertyValues {Curve {keys [
          {arriveTangent=2.4104, time=0, value=0}
          {arriveTangent=0.2537, interpMode=RCIM_Cubic, leaveTangent=0.2537, tangentMode=RCTM_User, time=1, value=1}
         ]}}})
    )
    Index 0 Param 4=1
    Index 2 Param 1=&FloatFromCurve(FloatCurve=DataInterface() {propertyValues {Curve {keys [<at 0.25 0.8 RCIM_Cubic>,<at 1 1>]}}})
    Index 2 Param 2=&FloatFromCurve(
     FloatCurve=DataInterface() {propertyValues {Curve {keys [<at 0 0.5 RCIM_Cubic>,<at 1 0 RCIM_Cubic>]}}}
     Scale Curve=&RandomRangeFloat(Minimum=0.2, Maximum=0.5)
    )
    Index 2 Param 3=&FloatFromCurve(FloatCurve=DataInterface() {propertyValues {Curve {keys [
         {interpMode=RCIM_Cubic, tangentMode=RCTM_Break, time=0, value=0.5}
         {interpMode=RCIM_Cubic, leaveTangent=11.1233, tangentMode=RCTM_Break, time=1, value=2}
        ]}}})
    Index 2 Param 4=&FloatFromCurve(
     FloatCurve=DataInterface() {propertyValues {Curve {keys [
         {arriveTangent=-0.4552, interpMode=RCIM_Cubic, leaveTangent=-0.4552, tangentMode=RCTM_User, time=0, value=1}
         {interpMode=RCIM_Cubic, leaveTangent=-1.2956, tangentMode=RCTM_Break, time=1, value=0.75}
        ]}}}
     Scale Curve=0.7
    )
    Module.Param0WriteEnabled=true
    Module.Param0WriteEnabled2=true
    Module.Param1WriteEnabled=false
    Module.Param1WriteEnabled2=true
    Module.Param2WriteEnabled=false
    Module.Param2WriteEnabled2=true
    Module.Param3WriteEnabled=false
    Module.Param3WriteEnabled2=true
    Write Parameter Index 2=true
   ) {module=DynamicMaterialParameters001}
   &ScaleSpriteSize(Scale Factor=&Vector2DFromFloat(Value=&FloatFromCurve(
      FloatCurve=DataInterface() {propertyValues {Curve {keys [
          {arriveTangent=3.7662, interpMode=RCIM_Cubic, leaveTangent=3.7662, tangentMode=RCTM_User, time=0, value=0.2}
          {arriveTangent=0.8, interpMode=RCIM_Cubic, leaveTangent=0.8, time=0.167, value=0.6111}
          {arriveTangent=0.1351, interpMode=RCIM_Cubic, leaveTangent=0.1351, tangentMode=RCTM_User, time=1, value=1}
         ]}}}
      Scale Curve=2
     )))
   &ScaleColor(
    Scale Alpha=&FloatFromCurve(FloatCurve=DataInterface() {propertyValues {Curve {keys [
         {arriveTangent=9191.0225, interpMode=RCIM_Cubic, time=0.0001, value=0.0123}
         {arriveTangentWeight=0.0484, interpMode=RCIM_Cubic, leaveTangentWeight=0.2906, tangentMode=RCTM_User, time=0.125, value=1}
         <at 1 0 RCIM_Cubic>
        ]}}})
    Scale RGB=&MakeVectorFromLinearColorRGB(Color=&ColorFromCurve(ColorCurve=DataInterface() {propertyValues {
        AlphaCurve {keys []}
        BlueCurve {keys [<at 0.023 10 RCIM_Cubic>,<at 0.4199 1 RCIM_Cubic>]}
        GreenCurve {keys [<at 0.023 4 RCIM_Cubic>,<at 0.4199 1 RCIM_Cubic>]}
        RedCurve {keys [<at 0.023 7.5974 RCIM_Cubic>,<at 0.4199 1 RCIM_Cubic>]}
       }}))
    ScaleA=true
    ScaleRGB=true
   )
   &SpriteRotationRate(Rotation Rate=&Multiply_Float(
     A=&Lerp_Float(A=-150, B=150, Alpha=$Particles.MaterialRandom)
     B=&Lerp_Float(A=1, B=0, Alpha=&Power(a=$Particles.NormalizedAge, b=0.7))
    ))
   &CameraOffset(Camera Offset Amount=/Niagara/DynamicInputs/Math/CalculateParticleRadius(
     Radius Calculation Type:ENiagaraCollisionRadiusOptions=::NewEnumerator0
     Particle Radius:NiagaraFloat=1
     Mesh Dimensions:Vector3f=<1 1 1>
     Method for Calculating Particle Radius:ENiagaraCalculateRadiusOptions=::NewEnumerator2
    )) {enabled=false}
   &WindForce(
    Air Resistance=0.25
    Wind Speed=$System.Direction
    Wind Speed Scale=&Multiply_Float(A=$System.Blowing, B=&RandomRangeFloat(Minimum=400, Maximum=500))
   ) {enabled=false, module=WindForce001}
   &SolveForcesAndVelocity(Clamp Velocity=false, Limit Acceleration=false)
  ]
  rendererClasses [&NiagaraSpriteRendererProperties]
  renderers [{rendererClass=&NiagaraSpriteRendererProperties}]
 }
 UmbrellaColoring {
  emitterUpdateScript [
   &EmitterState(Life Cycle Mode=::NewEnumerator1, Loop Behavior=::NewEnumerator1, Loop Duration=1, UseLoopDelay=true)
   &SpawnBurst_Instantaneous(Use Spawn Probability=false, UseLoopCountLimit=false)
  ]
  particleSpawnScript [
   &InitializeParticle(Lifetime=0.55, Sprite Size Mode=::NewEnumerator3, Uniform Sprite Size=170, UsePositionOffset=false, Write Material Random=true)
   &SkeletalMeshLocation(
    Bone Sampling Mode=::NewEnumerator4
    MaskSpawnGroup=false
    Sampled Position Offset=<0 0 -10>
    Skeletal Mesh=DataInterface() {propertyValues {
      DefaultMesh=null
      ExcludeBoneName=root
      OptionalComponentClass=null
      PreviewMesh=&Melancholy_Gadget
      SkinningMode=SkinOnTheFly
      SoftSourceActor=null
      SourceMode=Default
      UvSetIndex=0
      WholeMeshLOD=-1
      bExcludeBone=false
      bReadDeformedGeometry=true
      bRequireCurrentFrameData=true
      ComponentTags []
      FilteredBones [JNT_main_umbrella]
      FilteredSockets []
      MeshUserParameter {parameter {name=None, typeDefHandle {registeredTypeIndex=1}, varData []}}
      SamplingRegions []
     }}
   ) {enabled=false}
   /SetVariables(Particles.Position=&AddVectorToPosition(Position=$System.SampleSkeletalMesh.SampledPosition, Vector=<0 0 -12>)) {isSetParameters=true}
  ]
  particleUpdateScript [
   &ParticleState()
   &ScaleSpriteSize(Uniform Curve Sprite Scale=DataInterface() {propertyValues {Curve {keys [{interpMode=RCIM_Cubic, tangentMode=RCTM_User, time=0, value=0.2}, {interpMode=RCIM_Cubic, tangentMode=RCTM_User, time=0.4, value=1}]}}})
   &ScaleColor(
    Scale Alpha=&FloatFromCurve(FloatCurve=DataInterface() {propertyValues {Curve {keys [
         {interpMode=RCIM_Cubic, tangentMode=RCTM_User, time=0, value=1}
         {interpMode=RCIM_Cubic, tangentMode=RCTM_User, time=0.4, value=2}
         {interpMode=RCIM_Cubic, tangentMode=RCTM_User, time=1, value=0}
        ]}}})
    Scale RGB=&MakeVectorFromLinearColorRGB(Color=&ColorFromCurve(ColorCurve=DataInterface() {propertyValues {
        AlphaCurve {keys []}
        BlueCurve {keys [<at 0.0998 80>,<at 0.2424 0 RCIM_Cubic>]}
        GreenCurve {keys [<at 0.0998 0.8>,<at 0.2424 0 RCIM_Cubic>]}
        RedCurve {keys [<at 0.0998 45.1208>,<at 0.2424 0 RCIM_Cubic>]}
       }}))
    ScaleA=true
    ScaleRGB=true
   )
  ]
  rendererClasses [&NiagaraSpriteRendererProperties]
  renderers [
   {rendererClass=&NiagaraSpriteRendererProperties}
   {rendererClass=&NiagaraSpriteRendererProperties, rendererIndex=1}
  ]
 }
}
summary {
 systemName=NS_Emote_Melancholy_UmbrellaDissolve
 emitters [
  {emitterName=Ashes, rendererClasses [&NiagaraSpriteRendererProperties]}
  {emitterName=UmbrellaColoring, rendererClasses [&NiagaraSpriteRendererProperties]}
  {emitterName=Smoke, rendererClasses [&NiagaraSpriteRendererProperties]}
  {emitterName=AshesFalling, rendererClasses [&NiagaraSpriteRendererProperties]}
 ]
 userVariables [
  User.bIsFrontEnd {type=NiagaraBool}
  User.bIsFrontEndPreview {type=NiagaraBool}
  User.LightningStrikeActive {type=NiagaraFloat, defaultValue=0, description=""}
  User.NewNiagaraDataInterfaceSkeletalMesh {
   type=NiagaraDataInterfaceSkeletalMesh
   defaultValue {
    struct=NiagaraExt_VariableValue_DataInterface
    value {dataInterface=&self:NiagaraDataInterfaceSkeletalMesh_1}
   }
  }
  User.NormalizedNotifyProgress {type=NiagaraFloat, defaultValue=0, description=""}
  User.SkeletalMesh {
   type=NiagaraDataInterfaceSkeletalMesh
   defaultValue {
    struct=NiagaraExt_VariableValue_DataInterface
    value {dataInterface=&self:NiagaraDataInterfaceSkeletalMesh_0}
   }
  }
  User.UmbrellaUp {type=NiagaraFloat, defaultValue=0, description=""}
 ]
}
