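# Niagara system for the Melancholy emote that drips water off the rim of the umbrella prop.
# One sprite emitter spawns drops at filtered umbrella end bones (02 listed twice, no 04), with rate scaled by User.UmbrellaUp and randomized by spawn probability.
# Short-lived tinted drops fall under gravity, stretch with speed and fade through an alpha curve.
# The bIsFrontEnd size switch has identical values, so it changes nothing.
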NiagaraSystem /BRCosmetics/Animation/Game/MainPlayer/Emotes/Melancholy/VFX/NS_Emote_Melancholy_UmbrellaDrips
dependencies {
 usedDataInterfaces [&NiagaraDataInterfaceCurve]
 usedDynamicInputs []
 usedModules [&SampleSkeletalMesh, &SystemState]
}
emitters {Dripping {
  emitterUpdateScript [
   &EmitterState(Life Cycle Mode=::NewEnumerator1, Loop Delay=1, Loop Duration=4, UseLoopDelay=false)
   &SpawnRate(Spawn Probability=0.7, SpawnRate=&Multiply_Float(A=60, B=$User.UmbrellaUp), Use Spawn Probability=true)
  ]
  particleSpawnScript [
   &InitializeParticle(
    Color=&RandomRangeLinearColor(Minimum=<rgba 1.6 1.8453 2 1>, Maximum=<rgba 0.8 1.5027 2 1>)
    Lifetime=&RandomRangeFloat(Minimum=0.3, Maximum=0.5)
    Mass Mode=::NewEnumerator3
    Sprite Size=&Multiply_Vector2D_ByFloat(
     Float=&Multiply_Float(
      A=&MakeCustomFloatFromBool(Bool=$User.bIsFrontEnd, True Float=1, False Float=1)
      B=&RandomRangeFloat(Minimum=8, Maximum=12)
     )
     Vector2D=<0.1 0.5>
    )
    Sprite Size Mode=::NewEnumerator4
    UsePositionOffset=false
    Write Material Random=true
   )
   &AddVelocity(Velocity=&RandomRangeVector(Minimum=<0 0 -30>, Maximum=<0 0 -100>))
   &SkeletalMeshLocation(
    Bone Sampling Mode=::NewEnumerator0
    MaskSpawnGroup=false
    OverrideRandomSeed=false
    Skeletal Mesh=DataInterface() {propertyValues {
      DefaultMesh=null
      ExcludeBoneName=root
      OptionalComponentClass=null
      PreviewMesh=/BRCosmetics/Animation/Props/Gadgets_Melancholy/Meshes/Melancholy_Gadget
      SkinningMode=SkinOnTheFly
      SoftSourceActor=null
      SourceMode=Default
      UvSetIndex=0
      WholeMeshLOD=-1
      bExcludeBone=false
      bReadDeformedGeometry=true
      bRequireCurrentFrameData=true
      ComponentTags []
      FilteredBones [
       JNT_umbrella_01_end
       JNT_umbrella_02_end
       JNT_umbrella_02_end
       JNT_umbrella_03_end
       JNT_umbrella_05_end
       JNT_umbrella_06_end
       JNT_umbrella_07_end
       JNT_umbrella_08_end
       JNT_umbrella_09_end
       JNT_umbrella_10_end
      ]
      FilteredSockets []
      MeshUserParameter {parameter {name=None, typeDefHandle {registeredTypeIndex=1}, varData []}}
      SamplingRegions []
     }}
   )
   &ShapeLocation(MaskSpawnGroup=false, Override Seed=false, Shape Origin=$System.SampleSkeletalMesh.SampledPosition, Sphere Radius=5, UseOwnerScale=true) {enabled=false}
  ]
  particleUpdateScript [
   &ParticleState()
   &GravityForce(Gravity=<0 0 -300>)
   &SolveForcesAndVelocity(Clamp Velocity=false, Limit Acceleration=false)
   &ScaleSpriteSizeBySpeed(Max Scale Factor=<1 3>, Velocity Threshold=2000)
   &ScaleColor(
    Scale Alpha=&FloatFromCurve(FloatCurve=DataInterface() {propertyValues {Curve {keys [
         <at 0 0 RCIM_Cubic>
         {arriveTangent=0.0435, interpMode=RCIM_Cubic, leaveTangent=0.0435, tangentMode=RCTM_User, time=0.15, value=1}
         {arriveTangent=0.0176, interpMode=RCIM_Cubic, leaveTangent=0.0176, tangentMode=RCTM_User, time=0.6, value=1}
         <at 1 0 RCIM_Cubic>
        ]}}})
    ScaleA=true
    ScaleRGB=true
   )
  ]
  rendererClasses [&NiagaraSpriteRendererProperties]
  renderers [{rendererClass=&NiagaraSpriteRendererProperties}]
 }}
summary {
 systemName=NS_Emote_Melancholy_UmbrellaDrips
 emitters [{emitterName=Dripping, rendererClasses [&NiagaraSpriteRendererProperties]}]
 userVariables [
  User.bIsFrontEnd {type=NiagaraBool}
  User.bIsFrontEndPreview {type=NiagaraBool}
  User.NewNiagaraDataInterfaceSkeletalMesh {
   type=NiagaraDataInterfaceSkeletalMesh
   defaultValue {
    struct=NiagaraExt_VariableValue_DataInterface
    value {dataInterface=&self:NiagaraDataInterfaceSkeletalMesh_1}
   }
  }
  User.NormalizedNotifyProgress {type=NiagaraFloat, defaultValue=0, description=""}
  User.SkeletalMesh {
   type=NiagaraDataInterfaceSkeletalMesh
   defaultValue {
    struct=NiagaraExt_VariableValue_DataInterface
    value {dataInterface=&self:NiagaraDataInterfaceSkeletalMesh_0}
   }
  }
  User.UmbrellaUp {type=NiagaraFloat, defaultValue=0, description=""}
 ]
}
